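; Parent pawn component for ToyFox pilots that manages camera and input setup while attached to a Fortnite player pawn.
; BeginPlay applies a camera modifier and tagged camera mode override, adds a local input mapping context, and disables the spectate teleport camera; EndPlay removes modifier, override and mapping.
; View-target change bindings toggle the camera setup for local spectators.
; Input action tags become gameplay events, with authority relaying them toward simulated proxies.
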(event Custom|HandleViewTargetChanged
  (params (PlayerController :type Fort%20Player%20Controller%20Object%20Reference) (OldViewTarget :type Actor%20Object%20Reference) (NewViewTarget :type Actor%20Object%20Reference))
  (bind _returnvalue (Components|GetOwner))
  ; View Target Changed
  (if (== OldViewTarget _returnvalue)
    (CallFunction|SetCameraConfig PlayerController false true)
    (elif (== NewViewTarget _returnvalue)
      (CallFunction|SetCameraConfig PlayerController true))))

(event EventBeginPlay
  (bind _playercontroller (Variables|Default|GetPlayerController))
  ; Begin Play
  (|CallParentFunction)
  (Networking|CanExecuteCosmeticEvents
    (:True
      (bind _asfort_player_pawn_athena (Utilities|Casting|CastToFortPlayerPawnAthena (Components|GetOwner)))
      (Utilities|FlowControl|Sequence
        (:then_0
          ;; Disable local spectating's teleport phase camera overrides
          (Cinematics|Spectate|LocalDisableSpectateTeleportCamera (Actor|GetComponentbyClass (Fortnite|Athena|GameState|GetGameStateAthena) "/ToyFoxGameplay/Gameplay/B_ToyFoxGameStateComponent.B_ToyFoxGameStateComponent_C")))
        (:then_1
          ;; Future local spectating player
          (bind _output_get (Variables|Default|SetLocalViewControllers (Fortnite|GetLocalFortPlayerControllers true true)))
          (for _array_element _output_get
            (EventDispatchers|BindEventtoOnViewTargetChanged _array_element (EventDispatchers|CreateEvent :!event Custom|HandleViewTargetChanged))))
        (:then_2
          ;; Currently local spectating player's camera modifier
          (for _array_element_2 (FortPawn|GetAllLocalViewingPlayerControllers _asfort_player_pawn_athena)
            (CallFunction|SetCameraConfig _array_element_2 true)))
        (:then_3
          ;; Autonomous local player
          (bind _asfort_player_controller (Utilities|Casting|CastToFortPlayerController (Pawn|GetController _asfort_player_pawn_athena)))
          (bind _output_get_2 (Variables|Default|SetPlayerController _asfort_player_controller))
          ;;;[1 0.9188 0 0.5] Must reset here to avoid stuck from previous cinematics
          (Variables|/Script/Engine.PlayerCameraManager|SetbClientSimulatingViewTarget false (Variables|/Script/Engine.PlayerController|GetPlayerCameraManager _output_get_2))
          (CallFunction|SetCameraConfig _output_get_2 true)
          (if (Pawn|IsLocalPlayerController _playercontroller)
            (Input|AddMappingContext (PlayerController|LocalPlayerSubsystems|GetEnhancedInputLocalPlayerSubsystem _playercontroller) (Variables|Settings|GetInputMappingContext) (Variables|Settings|GetInputPriority) "(bIgnoreAllPressedKeysUntilRelease=True,bForceImmediately=False,bNotifyUserSettings=False)")))))))

(event EventEndPlay
  (params (EndPlayReason :type EEndPlayReason%20Enum :default "Destroyed"))
  (bind _playercontroller (Variables|Default|GetPlayerController))
  (bind _playercontroller_2 (Variables|Default|GetPlayerController))
  ; End play
  (Utilities|FlowControl|Sequence
    (:then_0
      (Networking|CanExecuteCosmeticEvents
        (:True
          (Utilities|FlowControl|Sequence
            (:then_0
              ;; Autonomous Local Player
              (Utilities|IsValid _playercontroller_2
                (:Is%20Valid
                  (CallFunction|SetCameraConfig _playercontroller_2)
                  ;;;[1 0.9188 0 0.5] Must reset again in case there was racing for begin play and cinematics over
                  (Variables|/Script/Engine.PlayerCameraManager|SetbClientSimulatingViewTarget false (Variables|/Script/Engine.PlayerController|GetPlayerCameraManager _playercontroller_2))
                  (if (Pawn|IsLocalPlayerController _playercontroller)
                    (Input|RemoveMappingContext (PlayerController|LocalPlayerSubsystems|GetEnhancedInputLocalPlayerSubsystem _playercontroller) (Variables|Settings|GetInputMappingContext) "(bIgnoreAllPressedKeysUntilRelease=True,bForceImmediately=False,bNotifyUserSettings=False)")))))
            (:then_1
              ;; Future local spectating player
              (bind _output_get (Variables|Default|SetLocalViewControllers (Fortnite|GetLocalFortPlayerControllers true true)))
              (for _array_element _output_get
                (EventDispatchers|UnbindEventfromOnViewTargetChanged _array_element (EventDispatchers|CreateEvent :!event Custom|HandleViewTargetChanged))))
            (:then_2
              ;; Current local spectating player
              (bind _asfort_player_pawn_athena (Utilities|Casting|CastToFortPlayerPawnAthena (Components|GetOwner)))
              (for _array_element_2 (FortPawn|GetAllLocalViewingPlayerControllers _asfort_player_pawn_athena)
                (CallFunction|SetCameraConfig _array_element_2)))))))
    (:then_1
      (|CallParentFunction EndPlayReason))))

(event Custom|DestroyComponent
  ; DestroyComponent: To enable k2 destroy component to be called successfully.
  (Networking|SwitchHasAuthority
    (:Authority
      (Components|DestroyComponent))))

(event Custom|BroadcastInputActionEventTag
  (params (GameplayTag :type Gameplay%20Tag%20Structure))
  ;[0.000717 0.15 0 0.5] Replicate input actions to other players.
  (Networking|SwitchHasAuthority
    (:Authority
      (CallFunction|ClientBroadcastInputActionEvent GameplayTag))
    (:Remote
      (Ability|SendGameplayEventtoActor (Components|GetOwner) GameplayTag))))

(event Custom|ClientBroadcastInputActionEvent
  (params (GameplayTag :type Gameplay%20Tag%20Structure))
  (bind _returnvalue (Components|GetOwner))
  ;[0.000717 0.15 0 0.5] Replicate input actions to other players.
  (Networking|CanExecuteCosmeticEvents
    (:True
      (Utilities|IsValid _returnvalue
        (:Is%20Valid
          (if (Utilities|Enum|Equal(Enum) (Networking|GetLocalRole _returnvalue) "ROLE_SimulatedProxy")
            (Ability|SendGameplayEventtoActor 0 GameplayTag)))))))

(fn FindOrAddGiantCameraModifer
  (params (PlayerController :type Player%20Controller%20Object%20Reference))
  (bind _cameramodifierclass (Variables|Settings|GetCameraModifierClass))
  ;[0.000114 0.119 0.15 0.5] FindOrAddGiantCameraModifier: Finding or adding desired camera modifier.
  (Utilities|FlowControl|Sequence
    (:then_0
      (Utilities|IsValid PlayerController
        (:Is%20Valid
          (if (Utilities|IsValidClass (Variables|Settings|GetCameraModifierClass))
            (bind _playercameramanager (Variables|/Script/Engine.PlayerController|GetPlayerCameraManager PlayerController)
              (:then
                (bind _returnvalue (Game|Player|FindCameraModifierbyClass _playercameramanager _cameramodifierclass))
                (Utilities|IsValid _returnvalue
                  (:Is%20Valid
                    (return _returnvalue))
                  (:Is%20Not%20Valid
                    (bind _returnvalue_2 (Game|Player|AddNewCameraModifier _playercameramanager _cameramodifierclass))
                    (return _returnvalue_2))))
              (:else))))))
    (:then_1
      (return 0))))

(fn SetCameraConfig
  (params (PlayerController :type Fort%20Player%20Controller%20Object%20Reference) (bEnableOverride :type Boolean) (bImmediate :type Boolean))
  (if bEnableOverride
    (bind _outmodifier (Default|FindorAddGiantCameraModifer PlayerController))
    (CameraModifier|EnableModifier _outmodifier)
    (CallFunction|SetCameraModeEnabled (Variables|Default|GetPlayerController) true)
    (else
      (bind _outmodifier_2 (Default|FindorAddGiantCameraModifer PlayerController))
      (CameraModifier|DisableModifier _outmodifier_2 (Variables|GetbImmediate))
      (CallFunction|SetCameraModeEnabled (Variables|Default|GetPlayerController)))))

(fn SetCameraModeEnabled
  (params (PlayerController :type Fort%20Player%20Controller%20Object%20Reference) (bEnabled :type Boolean))
  (if (and (Utilities|IsValidClass (Variables|Settings|GetCameraModeClass)) (GameplayTags|IsGameplayTagValid (Variables|Settings|GetCameraModeTag)))
    (bind _playercameramanager (Variables|/Script/Engine.PlayerController|GetPlayerCameraManager PlayerController)
      (:then
        (bind _asfort_player_camera_base (Utilities|Casting|CastToFortPlayerCameraBase _playercameramanager))
        (if (Variables|GetbEnabled)
          (Camera|AddCameraModeOverride _asfort_player_camera_base (Variables|Settings|GetCameraModeClass) (Variables|Settings|GetCameraModeTag) 0 0 0)
          (else
            (Camera|RemoveCameraModeOverride _asfort_player_camera_base (Variables|Settings|GetCameraModeTag) 0 false))))
      (:else))))
