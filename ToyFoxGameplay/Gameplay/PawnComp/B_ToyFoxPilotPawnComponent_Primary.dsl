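; Pawn component for the primary ToyFox pilot, adjusting camera setup for players spectating that pilot during the transition into the giant.
; SetCameraConfig calls the parent, then on enable, if the owner is a simulated proxy, sends a one-time ServerSpectateActor request to re-spectate it, forcing correct camera behavior.
; GetPlayerPawn returns the owner cast to FortPlayerPawn.
; Stomp and Scan input events are declared with no logic attached.
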(event EnhancedInputActionIA_ToyFox_Stomp
  (params (ActionValue :type Boolean) (ElapsedSeconds :type Float%20%28double-precision%29) (TriggeredSeconds :type Float%20%28double-precision%29) (InputAction :type Input%20Action%20Object%20Reference :default "/ToyFoxGameplay/Input/IA_ToyFox_Stomp.IA_ToyFox_Stomp")))

(event EnhancedInputActionIA_ToyFox_Scan
  (params (ActionValue :type Boolean) (ElapsedSeconds :type Float%20%28double-precision%29) (TriggeredSeconds :type Float%20%28double-precision%29) (InputAction :type Input%20Action%20Object%20Reference :default "/ToyFoxGameplay/Input/IA_ToyFox_Scan.IA_ToyFox_Scan")))

(fn GetPlayerPawn
  (bind _asfort_player_pawn (Utilities|Casting|CastToFortPlayerPawn (Components|GetOwner)))
  (return _asfort_player_pawn))

(fn SetCameraConfig
  (params (PlayerController :type Fort%20Player%20Controller%20Object%20Reference) (bEnableOverride :type Boolean) (bImmediate :type Boolean))
  (|CallParentFunction PlayerController bEnableOverride bImmediate)
  (if bEnableOverride
    ;[1 0.8508 0 0.5] Must reset spectate to force a correct camera behavior for player that spectate primary pilot before the transition into giant
    (if (and (Utilities|Enum|Equal(Enum) (Networking|GetLocalRole (Components|GetOwner)) "ROLE_SimulatedProxy") (not (Variables|Default|GetbSentRequest)))
      (bind _asfort_player_controller_zone (Utilities|Casting|CastToFortPlayerControllerZone PlayerController))
      (Team|ServerSpectateActor _asfort_player_controller_zone (Components|GetOwner) false)
      (Variables|Default|SetbSentRequest true))))
