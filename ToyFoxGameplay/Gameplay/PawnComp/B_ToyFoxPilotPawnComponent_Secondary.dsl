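; Pawn component for the ToyFox giant's secondary pilot, managing rocket ammo, inventory, power warning and camera state.
; On authority it stashes inventory until EndPlay, optionally force-revives from DBNO, and toggles a low-power gameplay effect from game state component events.
; Basic rocket ammo regenerates one round per timer tick, with hotfixable limits and a change dispatcher.
; Async tag listeners clear the camera-config flag during finisher watching; input handlers have no logic.
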(event EventBeginPlay
  ;[0 0.15 0.001494 0.5] BeginPlay: If desired, cache and remove inventory items.
  (|CallParentFunction)
  (Variables|Rocket|Basic|SetBasicRocketAmmoCount (Rocket|Basic|GetMaxBasicRocketAmmoCount))
  (Utilities|SwitchHasAuthority
    (:Authority
      (Utilities|FlowControl|Sequence
        (:then_0
          (CacheInventoryItem|CacheandRemoveMatchingInventoryItems)
          (if (> (ScalableFloat|GetCachedScalableFloatValue (Variables|Hotfix|GetHotfix_ForceReviveDBNO)) 0)
            (bind _asfort_player_pawn_athena (Utilities|Casting|CastToFortPlayerPawnAthena (Components|GetOwner)))
            (Class|FortPlayerPawnAthena|ForceRevivefromDBNO _asfort_player_pawn_athena (Pawn|GetController _asfort_player_pawn_athena))))
        (:then_1
          (bind _returnvalue (Fortnite|Athena|GameState|GetGameStateAthena))
          (Utilities|IsValid _returnvalue
            (:Is%20Valid
              (bind _returnvalue_2 (Actor|GetComponentbyClass _returnvalue "/ToyFoxGameplay/Gameplay/B_ToyFoxGameStateComponent.B_ToyFoxGameStateComponent_C"))
              (Utilities|IsValid _returnvalue_2
                (:Is%20Valid
                  (Variables|Transient|SetGameStateComponent _returnvalue_2)
                  (Default|BindEventtoOnGiantPowerLevelLowStateChanged _returnvalue_2 (EventDispatchers|CreateEvent :!event Custom|HandleGiantPowerLevelLowStateChanged))
                  (CallFunction|HandleGiantPowerLevelLowStateChanged (PowerLevel|GetPowerLevelLow _returnvalue_2)))))))))))

(event Custom|ReloadAmmoRocketBasic
  (bind _returnvalue (Rocket|Basic|GetMaxBasicRocketAmmoCount))
  ;[0 0.148 0.15 0.5] Reload rocket basic projectile ammo.
  (bind _output_get (Variables|Rocket|Basic|SetBasicRocketAmmoCount (Math|Integer|Clamp(Integer) (+ (Variables|Rocket|Basic|GetBasicRocketAmmoCount) 1) 0 _returnvalue)))
  (Default|CallOnBasicRocketAmmoCountChanged _output_get)
  (if (== _output_get _returnvalue)
    (Utilities|Time|ClearandInvalidateTimerbyHandle (Variables|Rocket|Basic|GetBasicRocketAmmoReloadTimerHandle))
    (else
      (Rocket|Basic|BeginBasicRocketAmmoReloadTimer))))

(event EventEndPlay
  (params (EndPlayReason :type EEndPlayReason%20Enum :default "Destroyed"))
  ;[0.125 0.0538 0 0.5] EndPlay: Clean up.
  (|CallParentFunction EndPlayReason)
  (Networking|SwitchHasAuthority
    (:Authority
      (CacheInventoryItem|RestoreCachedInventoryItems)
      (CallFunction|HandleGiantPowerLevelLowStateChanged false) ; Remove the applied gameplay effect.
      (bind _gamestatecomponent (Variables|Transient|GetGameStateComponent)
        (:then
          (Default|UnbindEventfromOnGiantPowerLevelLowStateChanged _gamestatecomponent (EventDispatchers|CreateEvent :!event Custom|HandleGiantPowerLevelLowStateChanged)))
        (:else)))))

(event Custom|SetupFinisherEventListening
  (params (CameraEnabled :type Boolean))
  ;[1 1 1] Set bCameraConfigEnabled to false when watching the finisher event
  (if CameraEnabled
    (bind _asynctaskproxy (Ability|Tasks|WaitGameplayTagAddToActor (Components|GetOwner) (Variables|Default|GetFinisherWatchingTag))
      (:then
        (Variables|Default|SetAsync_WaitFinisherTagAdd _asynctaskproxy)
        (bind _asynctaskproxy_2 (Ability|Tasks|WaitGameplayTagRemoveFromActor (Components|GetOwner) (Variables|Default|GetFinisherWatchingTag))
          (:then
            (Variables|Default|SetAsync_WaitFinisherTagRemove _asynctaskproxy_2))
          (:Removed
            (Variables|Default|SetbCameraConfigEnabled true))))
      (:Added
        (Variables|Default|SetbCameraConfigEnabled false)))
    (else
      (Ability|Async|EndAction (Variables|Default|GetAsync_WaitFinisherTagAdd)))))

(event EnhancedInputActionIA_ToyFox_Secondary_Rocket_Explosive
  (params (ActionValue :type Boolean) (ElapsedSeconds :type Float%20%28double-precision%29) (TriggeredSeconds :type Float%20%28double-precision%29) (InputAction :type Input%20Action%20Object%20Reference :default "/ToyFoxGameplay/Input/IA_ToyFox_Secondary_Rocket_Explosive.IA_ToyFox_Secondary_Rocket_Explosive")))

(event EnhancedInputActionIA_ToyFox_Secondary_Rocket_Scanner
  (params (ActionValue :type Boolean) (ElapsedSeconds :type Float%20%28double-precision%29) (TriggeredSeconds :type Float%20%28double-precision%29) (InputAction :type Input%20Action%20Object%20Reference :default "/ToyFoxGameplay/Input/IA_ToyFox_Secondary_Rocket_Scanner.IA_ToyFox_Secondary_Rocket_Scanner"))
  (bind _returnvalue (Components|GetOwner)))

(event EnhancedInputActionIA_ToyFox_Secondary_Exit
  (params (ActionValue :type Boolean) (ElapsedSeconds :type Float%20%28double-precision%29) (TriggeredSeconds :type Float%20%28double-precision%29) (InputAction :type Input%20Action%20Object%20Reference :default "/ToyFoxGameplay/Input/IA_ToyFox_Secondary_Exit.IA_ToyFox_Secondary_Exit"))
  (bind _returnvalue (Components|GetOwner)))

(event Custom|HandleGiantPowerLevelLowStateChanged
  (params (bLow :type Boolean))
  (bind _returnvalue (Components|GetOwner))
  (bind _returnvalue_2 (Components|GetOwner))
  (bind _returnvalue_3 (Ability|GetFortAbilitySystemComponent _returnvalue))
  (bind _returnvalue_4 (Ability|GetFortAbilitySystemComponent _returnvalue_2))
  ;[0.15 0.04002 0.003809 0.5] Server, Apply or remove the low power level warning gameplay effect based on the latest power level.
  (Utilities|SwitchHasAuthority
    (:Authority
      (if bLow
        (Utilities|IsValid _returnvalue_4
          (:Is%20Valid
            (GameplayEffects|ApplyGameplayEffectToSelf _returnvalue_4 (Variables|Settings|GetLowPowerLevelGameplayEffectClass) 1.0 (GameplayEffects|MakeEffectContext _returnvalue_4)) ; Sending in Level 1 to indicate secondary player.
          ))
        (else
          (Utilities|IsValid _returnvalue_3
            (:Is%20Valid
              (GameplayEffects|RemoveActiveGameplayEffectbySourceEffect _returnvalue_3 (Variables|Settings|GetLowPowerLevelGameplayEffectClass) 0 -1))))))))

(fn BeginBasicRocketAmmoReloadTimer
  ;[0 0.01318 0.15 0.5] Start the timer for reloading the ammo on the basic rocket.
  (bind _returnvalue (Utilities|Time|SetTimerbyEvent (EventDispatchers|CreateEvent :!event ReloadAmmoRocketBasic) (ScalableFloat|GetCachedScalableFloatValue (Variables|Hotfix|GetHotfix_BasicRocketAmmoRegenInterval))))
  (Variables|Rocket|Basic|SetBasicRocketAmmoReloadTimerHandle _returnvalue))

(fn ConsumeBasicRocketAmmo
  (params (AmmoCount :type Integer))
  ;[0 0.1062 0.15 0.5] Consume ammo of the basic rocket ammo.
  (if (> AmmoCount 0)
    (bind _output_get (Variables|Rocket|Basic|SetBasicRocketAmmoCount (Math|Integer|Clamp(Integer) (- (Variables|Rocket|Basic|GetBasicRocketAmmoCount) AmmoCount) 0 (Rocket|Basic|GetMaxBasicRocketAmmoCount))))
    (Default|CallOnBasicRocketAmmoCountChanged _output_get)
    (if (not (Utilities|Time|IsTimerActivebyHandle (Variables|Rocket|Basic|GetBasicRocketAmmoReloadTimerHandle)))
      (Rocket|Basic|BeginBasicRocketAmmoReloadTimer))))

(fn GetBasicRocketAmmoReloadTimerHandle
  (return (Variables|Rocket|Basic|GetBasicRocketAmmoReloadTimerHandle)))

(fn GetCameraConfigEnabled
  (return (Variables|Default|GetbCameraConfigEnabled)))

(fn GetCurrentBasicRocketAmmoCount
  (return (Variables|Rocket|Basic|GetBasicRocketAmmoCount)))

(fn GetMaxBasicRocketAmmoCount
  (return (Math|Float|Truncate (ScalableFloat|GetCachedScalableFloatValue (Variables|Hotfix|GetHotfix_BasicRocketMaxAmmoCount)))))

(fn SetCameraConfig
  (params (PlayerController :type Fort%20Player%20Controller%20Object%20Reference) (bEnableOverride :type Boolean) (bImmediate :type Boolean))
  (bind _returnvalue (Fortnite|Athena|GameState|GetGameStateAthena))
  (bind _returnvalue_2 (Actor|GetComponentbyClass _returnvalue "/ToyFoxGameplay/Gameplay/B_ToyFoxGameStateComponent.B_ToyFoxGameStateComponent_C"))
  ;[0 0.15 0.06251 0.5] SetCameraConfig: Update the bone visibility.
  (|CallParentFunction PlayerController bEnableOverride bImmediate)
  (Utilities|FlowControl|Sequence
    (:then_0
      (CallFunction|SetupFinisherEventListening (Variables|GetbEnableOverride)))
    (:then_1
      (Utilities|IsValid _returnvalue
        (:Is%20Valid
          (Utilities|IsValid _returnvalue_2
            (:Is%20Valid
              (bind _pawn (Giant|GetGiantPawn _returnvalue_2))
              (Utilities|IsValid _pawn
                (:Is%20Valid
                  (Variables|Default|SetbCameraConfigEnabled (Variables|GetbEnableOverride)))))))))))
